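# Temporary placeholder material for NeonCity signage props in Asteria, a Default Lit, Nanite-compatible surface with magenta base color and pale violet emissive parameters.
# An MF_Fresnel_FX function, tuned by brightness and exponent parameters, blends between low and high values, and the result multiplies the emissive color for a view-angle-dependent glow.
# Separate parameters define base color and roughness, alongside two unreferenced near-white color constants.
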Material /Game/Environments/Asteria/Props/NeonCity/NeonCity_Signage/Materials/Temp_Materials/TEMP_Emissive_03
graph {
 properties {bUsedWithNanite=true, pixelDepthOffsetMode=PDOM_Legacy, usedShadingModels=Default Lit}
 nodes [
  Constant3Vector() {constant=<rgba 0.984 1 0.9876 1>}
  Constant3Vector() {constant=<rgba 1 1 0.998 1>}
  n1 = ScalarParameter() {defaultValue=1, parameterName=EmissiveBrightness}
  n2 = ScalarParameter() {defaultValue=2, parameterName=EmissiveExponent}
  Multiply(
   A=VectorParameter() {parameterName=Emissive, defaultValue=<rgba 0.9179 0.6354 1 1>}.RGB
   B=LinearInterpolate(
    A=ScalarParameter() {defaultValue=-1, parameterName=FresnelLow}
    B=ScalarParameter() {defaultValue=1, parameterName=FresnelHigh}
    Alpha=MaterialFunctionCall(Brightness=@n1, Exponent=@n2) {materialFunction=&MF_Fresnel_FX}.Fresnel
   )
  )
  ScalarParameter() {defaultValue=0.8, parameterName=Roughness}
  VectorParameter() {parameterName=BaseColor, defaultValue=<rgba 0.3941 0.0198 0.3823 1>}
 ]
}
